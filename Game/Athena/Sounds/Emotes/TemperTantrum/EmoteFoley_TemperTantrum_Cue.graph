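# Sound cue for the Temper Tantrum emote's foley in Athena, mixing a main foley wave with looping rumble and steam-release layers.
# Both loops are mixed quieter and start after a shared fixed delay, with volume envelopes fading steam in quickly and rumble more slowly.
# It uses emote attenuation, the emote music sound class and concurrency with stop-oldest resolution, and has a music audio-analysis submix send.
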SoundCue /Game/Athena/Sounds/Emotes/TemperTantrum/EmoteFoley_TemperTantrum_Cue
root=Mixer_0
nodes [
 {id=Delay_0, children [Enveloper_2], properties {delayMax=1.16, delayMin=1.16}}
 {id=Delay_1, children [Enveloper_1], properties {delayMax=1.16, delayMin=1.16}}
 {
  id=Enveloper_1
  children [WavePlayer_1]
  properties {
   pitchCurve {editorCurveData {keys [{time=0}]}}
   volumeCurve {editorCurveData {keys [<at 0 0 RCIM_Cubic>, {arriveTangent=1.3561, interpMode=RCIM_Cubic, leaveTangent=1.3561, tangentMode=RCTM_User, time=2, value=1}]}}
  }
 }
 {
  id=Enveloper_2
  children [WavePlayer_2]
  properties {
   pitchCurve {editorCurveData {keys [{time=0}]}}
   volumeCurve {editorCurveData {keys [<at 0 0 RCIM_Cubic>, {arriveTangent=1.3561, interpMode=RCIM_Cubic, leaveTangent=1.3561, tangentMode=RCTM_User, time=0.2, value=1}]}}
  }
 }
 {id=Mixer_0, children [WavePlayer_0,Delay_0,Delay_1], properties {inputVolume [1,0.7,0.7]}}
 {id=WavePlayer_0, properties {soundWaveAssetPtr=&/EmoteFoley_TemperTantrum}}
 {id=WavePlayer_1, properties {bLooping=true, soundWaveAssetPtr=&/TemperTantrum_Rumble_Loop}}
 {id=WavePlayer_2, properties {bLooping=true, soundWaveAssetPtr=&/TemperTantrum_SteamRelease_Loop}}
]
properties {
 attenuationSettings=&Fort_Emotes_Athena_Att
 maxDistance=1600
 soundClassObject=&Fort_Music_Emote
 volumeMultiplier=0.8
 concurrencyOverrides {maxCount=3, resolutionRule=StopOldest}
 concurrencySet [/Game/Sounds/Concurrency_Rules/Athena_EmoteMusic_Conc]
 soundSubmixSends [{minSendLevel=0, soundSubmix=&Fortnite_AudioAnalysis_Music}]
}
